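# Niagara example system that draws a ribbon, named as a soft, flipped SubUV fire effect.
# A burst spawns a few particles that persist past their lifetime, with pendulum constraints whose lengths scale by execution index to space out the ribbon.
# Strong X acceleration pulls on the constrained points, while a camera vector sets each pivot offset.
# An incrementing counter feeds a dynamic material parameter; the exposed ChainLinks variable has no visible use.
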NiagaraSystem /Game/NeverCook/FX/Niagara_FX_Reuse_Examples/Ribbon/NS_Fire_SubUV_Ribbon_Flipped_Soft_Inst
dependencies {usedDataInterfaces [&NiagaraDataInterfaceCamera], usedDynamicInputs [], usedModules [&SystemState]}
emitters {Empty {
  emitterUpdateScript [
   &EmitterState(Life Cycle Mode=::NewEnumerator1, Loop Behavior=::NewEnumerator1, Loop Duration=1, UseLoopDelay=false)
   &SpawnBurst_Instantaneous(Spawn Count=5, Use Spawn Probability=false, UseLoopCountLimit=false)
  ]
  particleSpawnScript [
   &InitializeParticle(
    Color Mode=::NewEnumerator0
    Lifetime=0.2
    Ribbon Width=100
    Ribbon Width Mode=::NewEnumerator1
    UsePositionOffset=false
    Write Material Random=true
   )
   &AddVelocity(Velocity=<100 0 0>)
  ]
  particleUpdateScript [
   &ParticleState(Kill Particles When Lifetime Has Elapsed=false)
   /Niagara/Modules/Update/Utility/IncrementOverTime()
   &PendulumSetup(
    Module.Enable Spring Driver=false
    Pendulum Length:NiagaraFloat=&Lerp_Float(A=0, B=300, Alpha=&ReturnNormalizedExecIndex(Normalized Index Includes One=false))
    Pendulum Pivot Offset:Vector3f=&Multiply_VectorByFloat(
     Vector=&ReturnCameraVector(Camera Vectors=::NewEnumerator4, Coordinate Space=::World, Origin Position=$Particles.Position)
     Float=1
    )
    Pendulum Pivot Offset Coordinate Space:ENiagaraCoordinateSpace=::Simulation
    Rigid Pendulum=false
   )
   &DynamicMaterialParameters(
    Index 0 Param 1=$Particles.IncrementOverTime.Counter
    Module.Param0WriteEnabled=true
    Module.Param1WriteEnabled=false
    Module.Param2WriteEnabled=false
    Module.Param3WriteEnabled=false
   )
   &AccelerationForce(Acceleration=<50000 0 0>)
   &SolveForcesAndVelocity(Clamp Velocity=false, Limit Acceleration=false)
   &PendulumConstraint()
   &CalculateAccurateVelocity()
  ]
  rendererClasses [&NiagaraRibbonRendererProperties]
  renderers [{rendererClass=&NiagaraRibbonRendererProperties}]
 }}
summary {
 systemName=NS_Fire_SubUV_Ribbon_Flipped_Soft_Inst
 emitters [{emitterName=Empty, rendererClasses [&NiagaraRibbonRendererProperties]}]
 userVariables [User.ChainLinks {type=NiagaraInt32, defaultValue=3}]
}
