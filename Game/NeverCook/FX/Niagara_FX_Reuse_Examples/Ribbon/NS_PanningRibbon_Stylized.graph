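# Niagara example system of a stylized smoke ribbon trail, built from a single ribbon emitter.
# Particles spawn steadily, move with an initial velocity and panning curl noise, and partially inherit the source's movement.
# Over life the ribbon widens while curves shift its color and fade its alpha.
# Alpha is also scaled by emitter age and by owner speed, so the trail dims as the owner moves faster.
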NiagaraSystem /Game/NeverCook/FX/Niagara_FX_Reuse_Examples/Ribbon/NS_PanningRibbon_Stylized
dependencies {
 usedDataInterfaces [&NiagaraDataInterfaceCurve, &NiagaraDataInterfaceColorCurve]
 usedDynamicInputs []
 usedModules [&SystemState]
}
emitters {Smoke_Ribbin001 {
  emitterUpdateScript [
   &EmitterState(
    Enable Distance Culling=true
    Life Cycle Mode=::NewEnumerator1
    Loop Duration=1
    Max Distance Response=::SleepAndClearParticles
    MaxDistance=2000
    Module.bUseMaxDistance=true
    Module.bUseMinDistance=false
    Scalability Mode=::NewEnumerator1
    UseLoopDelay=false
   )
   &SpawnRate(SpawnRate=30, Use Spawn Probability=false)
  ]
  particleSpawnScript [
   &InitializeParticle(
    Color=<rgba 0.0591 0.1146 0.091 1>
    Ribbon Width=20
    Ribbon Width Mode=::NewEnumerator1
    Sprite Rotation Mode=::NewEnumerator2
    UsePositionOffset=false
    Write Material Random=true
   )
   &AddVelocity(Velocity=<50 0 0>) {module=AddVelocity002}
  ]
  particleUpdateScript [
   &ParticleState()
   &CurlNoiseForce(
    Noise Frequency=200
    Noise Quality / Cost=::NewEnumerator0
    Noise Strength=100
    Pan Noise=true
    Pan Noise Field=<-1 1 5>
    Randomization Vector:Vector3f=<0.64 0.45 0.2>
   )
   &InheritSourceMovement(Applied Position Scale=<0.8 0.8 0.9>, ApplyForce=false, ApplyPosition=true, ApplyPositionFalloff=false)
   &SolveForcesAndVelocity(Clamp Velocity=false, Limit Acceleration=false)
   &ScaleColor(
    Scale Alpha=&Multiply_Float(
     A=&FloatFromCurve(
      FloatCurve=DataInterface() {propertyValues {Curve {keys [
          {arriveTangent=1.6049, interpMode=RCIM_Cubic, leaveTangent=1.6049, tangentMode=RCTM_User, time=0, value=1.05}
          {arriveTangent=-2.4167, interpMode=RCIM_Cubic, leaveTangent=-2.4167, time=0.1647, value=1}
          {arriveTangent=-1.3646, interpMode=RCIM_Cubic, leaveTangent=-1.3646, tangentMode=RCTM_User, time=0.3, value=0.325}
          {tangentMode=RCTM_User, time=1, value=0}
         ]}}}
      Scale Curve=0.7
     )
     B=&FloatFromCurve(
      FloatCurve=DataInterface() {propertyValues {Curve {keys [<at 0.2 0 RCIM_Cubic>, {tangentMode=RCTM_User, time=0.8, value=1}]}}}
      CurveIndex=$Emitter.Age
      Scale Curve=&Multiply_Float(
       A=&RemapRange(
        Input Value To Remap=&VectorLength(Vector=$Engine.Owner.Velocity)
        Input Min=50
        Input Max=0
        Output Min=0.7
        Output Max=1
       )
       B=&RandomRangeFloat(Minimum=1, Maximum=1)
      )
     )
    )
    Scale RGB=&VectorFromFloat(Value=&FloatFromCurve(
      FloatCurve=DataInterface() {propertyValues {Curve {keys [<at 0 1.625 RCIM_Cubic>, {arriveTangent=-1.375, interpMode=RCIM_Cubic, leaveTangent=-1.375, time=0.5, value=1.625}, <at 1 0.25 RCIM_Cubic>]}}}
      Scale Curve=2
     ))
    ScaleA=true
    ScaleRGB=true
   )
   &ScaleRibbonWidth(Ribbon Width Scale=&FloatFromCurve(FloatCurve=DataInterface() {propertyValues {Curve {keys [
         {arriveTangent=0.0536, interpMode=RCIM_Cubic, leaveTangent=0.0536, tangentMode=RCTM_User, time=0, value=0.1}
         {arriveTangent=1.7358, interpMode=RCIM_Cubic, leaveTangent=1.7358, tangentMode=RCTM_User, time=0.2, value=0.25}
         {arriveTangent=0.9375, interpMode=RCIM_Cubic, leaveTangent=0.9375, time=0.3694, value=0.25}
         {arriveTangent=2.7752, interpMode=RCIM_Cubic, leaveTangent=2.7752, time=0.6, value=0.625}
         {arriveTangent=6.2786, interpMode=RCIM_Cubic, leaveTangent=6.2786, tangentMode=RCTM_User, time=1, value=2}
        ]}}}))
   &ScaleColor(
    Linear Color Curve=DataInterface() {propertyValues {
      AlphaCurve {keys [<at 0 1>]}
      BlueCurve {keys [<at 0.1932 0.5907>,<at 0.333 1>]}
      GreenCurve {keys [<at 0.1932 1>,<at 0.333 1>]}
      RedCurve {keys [<at 0.1932 0.8959>,<at 0.333 1>]}
     }}
    Scale Mode=::NewEnumerator2
   ) {module=ScaleColor001}
  ]
  rendererClasses [&NiagaraRibbonRendererProperties]
  renderers [{rendererClass=&NiagaraRibbonRendererProperties}]
 }}
summary {
 systemName=NS_PanningRibbon_Stylized
 emitters [{emitterName=Smoke_Ribbin001, rendererClasses [&NiagaraRibbonRendererProperties]}]
 userVariables []
}
