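# Niagara master ribbon example that trails short-lived, bright blue lightning ribbons from a skeletal mesh, filtered to the hand bones.
# A single LightningRibbon emitter uses custom FN ribbon modules to spawn, initialize and update ribbons against User.SkeletalMesh.
# Spawn-group masks split two ribbons with opposed offsets and phase-shifted sine/cosine spin.
# Added velocity pushes particles, curves fade width and alpha, and an accumulated emitter pan feeds dynamic material parameters.
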NiagaraSystem /Game/NeverCook/FX/Niagara_FX_Reuse_Examples/Ribbon/NS_Ribbon_Master_v3
dependencies {usedDataInterfaces [&NiagaraDataInterfaceCurve], usedDynamicInputs [], usedModules [&SystemState]}
emitters {LightningRibbon {
  emitterUpdateScript [
   &EmitterState(Life Cycle Mode=::NewEnumerator1, Loop Duration=1, UseLoopDelay=false)
   /Game/Effects/Niagara/Modules/Ribbons/FN_SpawnRibbon_v3(
    Number of Ribbons=2
    Skeleton Mesh=DataInterface() {propertyValues {
      DefaultMesh=null
      ExcludeBoneName=root
      OptionalComponentClass=null
      PreviewMesh=&M_Med_Soldier_04_Preview
      SkinningMode=SkinOnTheFly
      SoftSourceActor=null
      SourceMode=Default
      UvSetIndex=0
      WholeMeshLOD=-1
      bExcludeBone=false
      bReadDeformedGeometry=true
      bRequireCurrentFrameData=true
      ComponentTags []
      FilteredBones [hand_l,hand_r]
      FilteredSockets []
      MeshUserParameter {parameter {name=None, typeDefHandle {registeredTypeIndex=1}, varData []}}
      SamplingRegions []
     }}
    SpawnRate=30
   )
   /SetVariables(Emitter.Pan:NiagaraFloat=&Add_Float(A=$Emitter.Pan, B=&Multiply_Float(A=$Engine.DeltaTime, B=-2))) {isSetParameters=true}
  ]
  particleSpawnScript [
   &FN_SetSpawnGroup(Number of Spawn Groups=2)
   /Game/Effects/Niagara/Modules/Ribbons/FN_InitializeRibbon_v5(
    Color:LinearColor=<rgba 17.3438 46.2869 150 1>
    Custom Offset=true
    Enable Skeletal Position=true
    Lifetime=0.23
    Local Vector Offset:Vector3f=<1 0 0>
    Offset Amount=1
    Ribbon Width=65
    SkeletalMesh:NiagaraDataInterfaceSkeletalMesh=$User.SkeletalMesh
   )
   &self:InitializeRibbon_Spin(
    Color:LinearColor=<rgba 17.3438 46.2869 150 1>
    Enable Skeletal Position=true
    Lifetime=0.23
    MeshOffsetDistance=3
    MeshOffsetVector:Vector3f=&MakeVector(
     X=&Sine(
      Normalized Angle=&Add_Float(A=$Emitter.Age, B=&MaskFloatBySpawnGroup(Masked Float=0, Passthrough Float=0.12))
      Period=0.4
      Scale=&MaskFloatBySpawnGroup(Masked Float=-1, Passthrough Float=1)
      Bias=0
     )
     Y=0
     Z=&Cosine(
      Normalized Angle=&Add_Float(A=$Emitter.Age, B=&MaskFloatBySpawnGroup(Masked Float=0, Passthrough Float=0.12))
      Period=0.4
      Scale=1
      Bias=0
     )
    )
    Ribbon Width=65
    SkeletalMesh:NiagaraDataInterfaceSkeletalMesh=$User.SkeletalMesh
   ) {module=InitializeRibbon_Spin}
   &AddVelocity(Velocity=&FN_FrontEndVector3Modifier(Gameplay=<500 0 0>, Front End=<500 0 0>)) {module=AddVelocity001}
   &DynamicMaterialParameters(
    Index 0 Param 1=1
    Index 0 Param 2=&MaskFloatBySpawnGroup(Masked Float=$Emitter.Pan, Passthrough Float=&Add_Float(A=$Emitter.Pan, B=0.5))
    Index 0 Param 3=1
    Index 0 Param 4=&MaskFloatBySpawnGroup(Masked Float=$Emitter.Pan, Passthrough Float=&Add_Float(A=$Emitter.Pan, B=0.5))
    Index 1 Param 1=1
    Index 1 Param 2=1
    Index 1 Param 3=1
    Index 1 Param 4=1
    Index 2 Param 1=1
    Index 2 Param 2=1
    Index 2 Param 3=1
    Index 2 Param 4=1
    Index 3 Param 1=1
    Index 3 Param 2=1
    Index 3 Param 3=1
    Module.Param0WriteEnabled=true
    Module.Param0WriteEnabled1=true
    Module.Param0WriteEnabled2=true
    Module.Param0WriteEnabled3=true
    Module.Param1WriteEnabled=true
    Module.Param1WriteEnabled1=true
    Module.Param1WriteEnabled2=true
    Module.Param1WriteEnabled3=true
    Module.Param2WriteEnabled=true
    Module.Param2WriteEnabled1=true
    Module.Param2WriteEnabled2=true
    Module.Param2WriteEnabled3=true
    Module.Param3WriteEnabled=true
    Module.Param3WriteEnabled1=true
    Module.Param3WriteEnabled2=true
    Module.Param3WriteEnabled3=true
    Write Parameter Index 1=true
    Write Parameter Index 2=true
    Write Parameter Index 3=true
   )
   /SetVariables(Particles.Position=&Add_Vector(
     A=Linked() {linkedVariable {name=Particles.Position, type=NiagaraPosition}}
     B=&Multiply_VectorByFloat(
      Vector=&RotateVectorByQuaternion(
       Vector To Rotate=&MaskVectorBySpawnGroup(Masked Vector=<0 -1 0>, Passthrough Vector=<0 1 0>)
       Quaternion=$Particles.MeshOrientation
      )
      Float=15
     )
    )) {isSetParameters=true}
  ]
  particleUpdateScript [
   &SolveForcesAndVelocity(Clamp Velocity=false, Limit Acceleration=false)
   /Game/Effects/Niagara/Modules/Ribbons/FN_UpdateRibbon_v6(Enable Skeletal Position=true, SkeletalMesh:NiagaraDataInterfaceSkeletalMesh=$User.SkeletalMesh)
   &ScaleRibbonWidth(Ribbon Width Scale=&FloatFromCurve(
     FloatCurve=DataInterface() {propertyValues {Curve {keys [
         {arriveTangent=9.0725, interpMode=RCIM_Cubic, leaveTangent=9.0725, tangentMode=RCTM_User, time=0, value=0}
         <at 0.1408 0.9847 RCIM_Cubic>
         <at 0.5 0.95>
         {arriveTangent=-0.0405, interpMode=RCIM_Cubic, leaveTangent=-0.0405, tangentMode=RCTM_User, time=1, value=0}
        ]}}}
     Scale Curve=0.5
    ))
   &ScaleColor(
    Scale Alpha=&FloatFromCurve(FloatCurve=DataInterface() {propertyValues {Curve {keys [
         {arriveTangent=1.5518, interpMode=RCIM_Cubic, leaveTangent=1.5518, tangentMode=RCTM_User, time=0.1, value=0}
         {arriveTangent=-2.5249, interpMode=RCIM_Cubic, leaveTangent=-2.5249, tangentMode=RCTM_User, time=0.6, value=0.8}
         {arriveTangent=-0.0405, interpMode=RCIM_Cubic, leaveTangent=-0.0405, tangentMode=RCTM_User, time=1, value=0}
        ]}}})
    ScaleA=true
    ScaleRGB=true
   )
  ]
  rendererClasses [&NiagaraRibbonRendererProperties]
  renderers [{rendererClass=&NiagaraRibbonRendererProperties}]
 }}
summary {
 systemName=NS_Ribbon_Master_v3
 emitters [{emitterName=LightningRibbon, rendererClasses [&NiagaraRibbonRendererProperties]}]
 userVariables [
  User.bIsFrontEnd {type=NiagaraBool}
  User.bIsFrontEndPreview {type=NiagaraBool}
  User.RibbonIDOffset {type=NiagaraInt32}
  User.SkeletalMesh {
   type=NiagaraDataInterfaceSkeletalMesh
   defaultValue {
    struct=NiagaraExt_VariableValue_DataInterface
    value {dataInterface=&self:NiagaraDataInterfaceSkeletalMesh_0}
   }
  }
 ]
}
